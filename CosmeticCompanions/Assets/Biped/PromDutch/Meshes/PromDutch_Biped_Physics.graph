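# Physics asset for the PromDutch biped cosmetic companion skeletal mesh.
# It holds a single capsule body on the pelvis, offset and rotated along the bone to approximate the character with one shape.
# It defines no constraints, so the other bones have no physics bodies and the asset does not support an articulated ragdoll.
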PhysicsAsset /CosmeticCompanions/Assets/Biped/PromDutch/Meshes/PromDutch_Biped_Physics
bodies {pelvis [{shapeType=Capsule, length=14.7242, radius=24.3268, center=<12.9589 0 0>, rotation=<rot 90 0 0>}]}
constraints []
